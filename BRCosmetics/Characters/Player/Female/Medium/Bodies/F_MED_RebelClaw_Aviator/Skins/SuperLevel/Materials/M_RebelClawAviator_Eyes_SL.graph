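# Eye material for the Super Level skin of the Rebel Claw Aviator character, built on the shared FNEye eye function wrapped in TechArt overrides.
# It extracts that result's base color, then applies a hue shift, desaturation and a brightness power curve, each driven by a SuperLevel parameter.
# The adjusted color replaces only Base Color in the material attributes, leaving the rest of the eye shading intact.
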Material /BRCosmetics/Characters/Player/Female/Medium/Bodies/F_MED_RebelClaw_Aviator/Skins/SuperLevel/Materials/M_RebelClawAviator_Eyes_SL
graph {
 properties {
  bUseMaterialAttributes=true
  bUsedWithMeshDeformer=true
  bUsedWithSkeletalMesh=true
  pixelDepthOffsetMode=PDOM_Legacy
  shadingModel=MSM_FromMaterialExpression
  usedShadingModels=Default Lit
 }
 nodes [
  n1 = MaterialFunctionCall(In=MaterialFunctionCall() {materialFunction=/Game/Characters/Shared/FNEye/Materials/MF_eyes}.Result) {materialFunction=&MF_TechArtOverrides}
  n2 = MaterialFunctionCall(
   Hue Shift Percentage=ScalarParameter() {defaultValue=0.5, group=SuperLevel, parameterName=HueShift}
   Texture=GetMaterialAttributes(None=@n1.Result) {attributeGetTypes [#1]}.Base Color
  ) {materialFunction=&HueShift}
  n3 = ScalarParameter() {defaultValue=0.2, group=SuperLevel, parameterName=Desaturation}
  SetMaterialAttributes!(
   MaterialAttributes=@n1.Result
   Base Color=Power(Base=Desaturation(None=@n2.Result, Fraction=@n3), Exp=ScalarParameter() {defaultValue=1, group=SuperLevel, parameterName=Brightness})
  ) {attributeSetTypes [#1]}
 ]
}
