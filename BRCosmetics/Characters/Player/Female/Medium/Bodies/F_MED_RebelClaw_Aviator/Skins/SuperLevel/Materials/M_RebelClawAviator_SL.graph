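# Skinned character material for the super level skin of the female RebelClaw Aviator, built on the MF_CharacterParent_2023 PreFX and PostFX functions.
# An IsHead static switch sends non-head surfaces through the shared MF_S27SL super level effect.
# On the head, dark base-color pixels (red channel) inside a scaled-UV face-paint mask become a strip with parameterized color, metallic and roughness, requiring no mask texture in this graph.
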Material /BRCosmetics/Characters/Player/Female/Medium/Bodies/F_MED_RebelClaw_Aviator/Skins/SuperLevel/Materials/M_RebelClawAviator_SL
graph {
 properties {
  bUseMaterialAttributes=true
  bUsedWithMeshDeformer=true
  bUsedWithMorphTargets=true
  bUsedWithSkeletalMesh=true
  pixelDepthOffsetMode=PDOM_Legacy
  shadingModel=MSM_FromMaterialExpression
  usedShadingModels=Default Lit | Subsurface | Clear Coat | Subsurface Profile | Cloth
 }
 nodes [
  Composite() {subgraphName=FacePaintMask}
  PinBase()
  n1 = MaterialFunctionCall(UVs=Add(A=TextureCoordinate(), B=Constant2Vector() {g=0.05}), Texture Scale=Constant2Vector() {g=0.1, r=0.15}) {materialFunction=&ScaleUVsByCenter}
  PinBase!(Mask=@n1.0-1 mask) {reroutePins [Mask {expression=&self:MaterialExpressionReroute_0}]}
  n2 = VectorParameter() {group=SuperLevel, parameterName=StripColor, defaultValue=<rgba 0.3559 0.2215 0.0117 1>}
  n3 = MaterialFunctionCall() {materialFunction=&MF_CharacterParent_2023_PreFX}
  n4 = MaterialFunctionCall(Material=@n3.Result, M=@n3.Result, Pre Skinned Local Position=@n3.Result, Distance Blend=@n3.Result, Fresnel=@n3.Result) {materialFunction=&MF_CharacterParent_2023_PostFX}
  n5 = GetMaterialAttributes(None=@n4.Result) {attributeGetTypes [#1, #2, #3]}
  n6 = ComponentMask(None=@n5.Base Color) {r=true}
  MaterialExpressionReroute_0 = Reroute(None=@n1.0-1 mask)
  n7 = SmoothStep(Value=LinearInterpolate(B=Multiply(A=@n6, B=@MaterialExpressionReroute_0.Mask), Alpha=@MaterialExpressionReroute_0.Mask) {constA=0.32}) {constMax=0.25, constMin=0.1}
  n8 = ScalarParameter() {defaultValue=1, group=SuperLevel, parameterName=StripMetallic}
  n9 = ScalarParameter() {defaultValue=0.2, group=SuperLevel, parameterName=StripRoughness}
  StaticSwitchParameter(
   True=SetMaterialAttributes!(
    MaterialAttributes
    Base Color=LinearInterpolate(A=@n2.RGB, B=@n5.Base Color, Alpha=@n7)
    Metallic=LinearInterpolate(A=@n8, B=@n5.Metallic, Alpha=@n7)
    Roughness=LinearInterpolate(A=@n9, B=@n5.Roughness, Alpha=@n7)
   ) {attributeSetTypes [#1, #2, #3]}
   False=MaterialFunctionCall!(InMA=@n4.Result, FXMask=@n3.Tech Art Mask, UseOriginalEmissive, PreSkinPos=@n3.Tech Art Mask) {materialFunction=/Game/Characters/Player/Common/MaterialFunctions/Superlevel_S27/MF_S27SL}.OutMA
  ) {parameterName=IsHead}
 ]
}
